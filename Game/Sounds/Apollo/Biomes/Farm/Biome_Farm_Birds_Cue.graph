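# Sound cue for one-shot farm biome bird ambience in the outdoor Athena ambient sound class.
# A Random node picks with equal weight among eleven BirdB wave variations.
# A Modulator at the root slightly varies pitch and volume on each play to avoid repetition.
# Playback uses the shared tile-ambient one-shot attenuation and a reduced overall volume.
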SoundCue /Game/Sounds/Apollo/Biomes/Farm/Biome_Farm_Birds_Cue
root=Modulator_0
nodes [
 {id=Modulator_0, children [Random_1], properties {pitchMax=1.1, pitchMin=0.9, volumeMax=1.2, volumeMin=0.8}}
 {
  id=Random_1
  children [
   WavePlayer_6
   WavePlayer_9
   WavePlayer_15
   WavePlayer_16
   WavePlayer_17
   WavePlayer_18
   WavePlayer_19
   WavePlayer_20
   WavePlayer_21
   WavePlayer_22
   WavePlayer_23
  ]
  properties {weights [1,1,1,1,1,1,1,1,1,1,1]}
 }
 {id=WavePlayer_15, properties {soundWaveAssetPtr=&/Biome_Farm_BirdB_01}}
 {id=WavePlayer_16, properties {soundWaveAssetPtr=&/Biome_Farm_BirdB_02}}
 {id=WavePlayer_17, properties {soundWaveAssetPtr=&/Biome_Farm_BirdB_03}}
 {id=WavePlayer_18, properties {soundWaveAssetPtr=&/Biome_Farm_BirdB_04}}
 {id=WavePlayer_19, properties {soundWaveAssetPtr=&/Biome_Farm_BirdB_05}}
 {id=WavePlayer_20, properties {soundWaveAssetPtr=&/Biome_Farm_BirdB_06}}
 {id=WavePlayer_21, properties {soundWaveAssetPtr=&/Biome_Farm_BirdB_07}}
 {id=WavePlayer_22, properties {soundWaveAssetPtr=&/Biome_Farm_BirdB_08}}
 {id=WavePlayer_23, properties {soundWaveAssetPtr=&/Biome_Farm_BirdB_09}}
 {id=WavePlayer_6, properties {soundWaveAssetPtr=&/Biome_Farm_BirdB_10}}
 {id=WavePlayer_9, properties {soundWaveAssetPtr=&/Biome_Farm_BirdB_11}}
]
properties {
 attenuationSettings=/Game/Sounds/Attenuation/Fort_TileAmbient_OneShot_Att
 maxDistance=8000
 soundClassObject=&Fort_Ambient_Outdoors_Athena
 volumeMultiplier=0.501
 preEffectBusSends [{}]
}
