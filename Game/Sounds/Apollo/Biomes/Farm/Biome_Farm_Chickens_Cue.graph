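# Sound cue for farm biome chicken ambience, mixing chicken sounds with iron gate sounds borrowed from the Town biome.
# A Random node picks with equal weight among seven chicken waves and six iron gate waves, so gates are nearly half the picks.
# A root Modulator varies pitch and volume on each play.
# The cue plays quietly through the outdoor ambient sound class with tile ambient one-shot attenuation.
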SoundCue /Game/Sounds/Apollo/Biomes/Farm/Biome_Farm_Chickens_Cue
root=Modulator_0
nodes [
 {id=Modulator_0, children [Random_1], properties {pitchMax=1.1, pitchMin=0.9, volumeMax=1.2, volumeMin=0.8}}
 {
  id=Random_1
  children [
   WavePlayer_6
   WavePlayer_9
   WavePlayer_11
   WavePlayer_12
   WavePlayer_13
   WavePlayer_14
   WavePlayer_15
   WavePlayer_5
   WavePlayer_7
   WavePlayer_8
   WavePlayer_10
   WavePlayer_19
   WavePlayer_20
  ]
  properties {weights [1,1,1,1,1,1,1,1,1,1,1,1,1]}
 }
 {id=WavePlayer_10, properties {soundWaveAssetPtr=/Game/Sounds/Apollo/Biomes/Town/Biome_Town_IronGate_04}}
 {id=WavePlayer_11, properties {soundWaveAssetPtr=&/Biome_Farm_Chicken_01}}
 {id=WavePlayer_12, properties {soundWaveAssetPtr=&/Biome_Farm_Chicken_10}}
 {id=WavePlayer_13, properties {soundWaveAssetPtr=&/Biome_Farm_Chicken_07}}
 {id=WavePlayer_14, properties {soundWaveAssetPtr=&/Biome_Farm_Chicken_08}}
 {id=WavePlayer_15, properties {soundWaveAssetPtr=&/Biome_Farm_Chicken_09}}
 {id=WavePlayer_19, properties {soundWaveAssetPtr=/Game/Sounds/Apollo/Biomes/Town/Biome_Town_IronGate_05}}
 {id=WavePlayer_20, properties {soundWaveAssetPtr=/Game/Sounds/Apollo/Biomes/Town/Biome_Town_IronGate_06}}
 {id=WavePlayer_5, properties {soundWaveAssetPtr=/Game/Sounds/Apollo/Biomes/Town/Biome_Town_IronGate_01}}
 {id=WavePlayer_6, properties {soundWaveAssetPtr=&/Biome_Farm_Chicken_06}}
 {id=WavePlayer_7, properties {soundWaveAssetPtr=/Game/Sounds/Apollo/Biomes/Town/Biome_Town_IronGate_02}}
 {id=WavePlayer_8, properties {soundWaveAssetPtr=/Game/Sounds/Apollo/Biomes/Town/Biome_Town_IronGate_03}}
 {id=WavePlayer_9, properties {soundWaveAssetPtr=&/Biome_Farm_Chicken_02}}
]
properties {
 attenuationSettings=/Game/Sounds/Attenuation/Fort_TileAmbient_OneShot_Att
 maxDistance=8000
 soundClassObject=&Fort_Ambient_Outdoors_Athena
 volumeMultiplier=0.301
 preEffectBusSends [{}]
}
